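; Projectile blueprint that triggers an explosion shortly after spawning, a base for on-death effect gameplay modifiers going by its name and folder.
; On BeginPlay it waits a random delay of 0.05 to 0.2 seconds, then calls DoExplosion from the Delivery category.
; The random delay staggers detonations, and the construction script is empty.
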(event EventBeginPlay
  (Utilities|FlowControl|Delay (Math|Random|RandomFloatinRange 0.05 0.2))
  (Delivery|DoExplosion))

(fn ConstructionScript)
